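; Spline-based environment Blueprint that lines a spline with instanced copies of one mesh, apparently wave pieces for KelpLinen.
; The construction script rebuilds the instances, spacing them by scaled mesh X bounds plus an offset and aiming each at the next point.
; Set_Materials applies a dynamic material with texture, color and surface parameters to the whole component, redundantly on every loop iteration.
; BeginPlay, overlap and Tick events are empty.
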(event EventBeginPlay)

(event Collision|EventActorBeginOverlap
  (params (OtherActor :type Actor%20Object%20Reference)))

(event EventTick
  (params (DeltaSeconds :type Float%20%28single-precision%29)))

(fn GetLocationAtIndex
  (params (A :type Integer))
  (return (Spline|GetLocationatDistanceAlongSpline (Variables|Default|GetSpline) (* A (Variables|Default|GetSpacing)) "Local")))

(fn Set_Materials
  (params (Mesh :type Mesh%20Component%20Object%20Reference))
  (bind _returnvalue (Rendering|Material|CreateDynamicMaterialInstance Mesh 0 (Variables|Default|GetSource%20Material) "None" :!class /Script/Engine.PrimitiveComponent))
  (bind _output_get (Variables|Default|SetDynamic%20material _returnvalue))
  (Rendering|Material|SetTextureParameterValue _output_get "Normal" (Variables|Textures|GetNormal))
  (Rendering|Material|SetTextureParameterValue _output_get "DetailsNormal" (Variables|Textures|GetDetails%20Normal))
  (Rendering|Material|SetTextureParameterValue _output_get "Specular" (Variables|Textures|GetSpecMask))
  (Rendering|Material|SetTextureParameterValue _output_get "Diffuse" (Variables|Textures|GetDiffuse))
  (Rendering|Material|SetVectorParameterValueonMaterials Mesh "DecalColor" (Math|Conversions|ToVector(LinearColor) (Variables|Colors|GetColor%20Top)))
  (Rendering|Material|SetVectorParameterValueonMaterials Mesh "DecalColor2" (Math|Conversions|ToVector(LinearColor) (Variables|Colors|GetColor%20Bottom)))
  (Rendering|Material|SetScalarParameterValueonMaterials Mesh "ScaleTextures" (Variables|Colors|GetScaleTextures))
  (Rendering|Material|SetScalarParameterValueonMaterials Mesh "MaskDensity" (Variables|Colors|GetMaskDensity))
  (Rendering|Material|SetScalarParameterValueonMaterials Mesh "RoughnessTweak" (Variables|Colors|GetRoughness%20Top))
  (Rendering|Material|SetScalarParameterValueonMaterials Mesh "SpecularTweak" (Variables|Colors|GetSpecular%20Top))
  (Rendering|Material|SetScalarParameterValueonMaterials Mesh "RoughnessTweak2" (Variables|Colors|GetRoughness%20Bottom))
  (Rendering|Material|SetScalarParameterValueonMaterials Mesh "ColorTransition " (Variables|Colors|GetColorTransition%20))
  (Rendering|Material|SetScalarParameterValueonMaterials Mesh "ColorContrast" (Variables|Colors|GetColorContrast))
  (Rendering|Material|SetScalarParameterValueonMaterials Mesh "DetailsValue" (Variables|Colors|GetDetailsValue))
  (Rendering|Material|SetScalarParameterValueonMaterials Mesh "DetailsScale" (Variables|Colors|GetDetailsScale))
  (Rendering|Material|SetScalarParameterValueonMaterials Mesh "SpecularTweak2" (Variables|Colors|GetSpecular%20Bottom))
  (Rendering|Material|SetScalarParameterValueonMaterials Mesh "Blendin" (Variables|Colors|GetBlend%20In%20))
  (Rendering|Material|SetScalarParameterValueonMaterials Mesh "Blend In Contrast" (Variables|Colors|GetBlend%20In%20Contrast))
  (Rendering|Material|SetScalarParameterValueonMaterials Mesh "Nomal Map Value" (Variables|Colors|GetNomal%20Map%20Value)))

(fn ConstructionScript
  (bind (_min _max) (Utilities|Struct|BreakBox (StaticMesh|GetBoundingBox (Variables|Default|GetCurrent%20Mesh))))
  (bind _current_mesh (Variables|Default|GetCurrent%20Mesh))
  (Components|InstancedStaticMesh|ClearInstances (Variables|Default|GetInstancedStaticMesh))
  (Components|StaticMesh|SetStaticMesh (Variables|Default|GetInstancedStaticMesh) _current_mesh)
  (bind _output_get (Variables|Default|SetSpacing (+ (.x (- (* _max (Variables|Default|GetScale)) (* _min (Variables|Default|GetScale)))) (Variables|Default|GetOffset))))
  (bind _output_get_2 (LocalVariables|Setnumber%20of%20instances (Math|Float|Floor (/ (Spline|GetSplineLength (Variables|Default|GetSpline)) _output_get))))
  (for _index (range (+ _output_get_2 1))
    (bind _returnvalue (Components|InstancedStaticMesh|AddInstance (Variables|Default|GetInstancedStaticMesh) (Math|Transform|MakeTransform (CallFunction|GetLocationatIndex _index) (Math|Rotator|MakeRotfromX (- (CallFunction|GetLocationatIndex (+ _index 1)) (CallFunction|GetLocationatIndex _index))) (Math|Conversions|ToVector(Float) (Variables|Default|GetScale))) false))
    (CallFunction|SetMaterials (Variables|Default|GetInstancedStaticMesh))))
