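# Translucent decal material for KelpLinen terrain, built for instanced meshes, that writes color, surface attributes and displacement into a Runtime Virtual Texture, with a non-VT fallback.
# Cliff textures sample through randomly rotated UVs or top-down world projection, feeding height-lerped tint colors and detail-blended normals.
# Opacity is either parameter-driven or a height-blended radial falloff with per-instance density, multiplied by beach gradients.
# Mobile desaturates the RVT base color.
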Material /Clyde/Environments/KelpLinen/BP/M_RVT_Decal_Blueprint_KelpLinen
graph {
 properties {
  bAutomaticallySetUsageInEditor=false
  bCastRayTracedShadows=false
  bTangentSpaceNormal=false
  bUseMaterialAttributes=true
  bUsedWithInstancedStaticMeshes=true
  blendMode=BLEND_Translucent
  pixelDepthOffsetMode=PDOM_Legacy
  usedShadingModels=Default Lit
 }
 nodes [
  n1 = ScalarParameter() {parameterName=DetailsValue}
  n2 = Multiply(A=NamedRerouteUsage(), B=AppendVector(A=AppendVector(A=@n1, B=@n1), B=Constant() {r=1}))
  n3 = Constant2Vector() {g=1, r=1}
  n4 = Multiply(A=@n3, B=ScalarParameter() {defaultValue=1, parameterName=Nomal Map Value})
  n5 = MaterialFunctionCall(
   A=VectorParameter() {parameterName=DecalColor2, defaultValue=<rgba 0.5 0.5 0.5 0>}.RGB
   B=LinearInterpolate(
    A=VectorParameter() {parameterName=DecalColor, defaultValue=<rgba 0.5 0.5 0.5 0>}.RGB
    B=VectorParameter() {parameterName=Secondary_Glow, defaultValue=<rgba 0.6199 0 1 1>}.RGB
    Alpha=NamedRerouteUsage()
   )
   Transition Phase=ScalarParameter() {defaultValue=0.5, parameterName="ColorTransition "}
   Height Texture=NamedRerouteUsage()
   Contrast=ScalarParameter() {parameterName=ColorContrast}
  ) {materialFunction=&HeightLerp}
  n6 = StaticBoolParameter() {defaultValue=true, parameterName=UseHeightChannelForVariation}
  n7 = StaticSwitch(True=@n5.Alpha, False=Constant(), Value=@n6)
  n8 = Multiply(A=NamedRerouteUsage(), B=AppendVector(A=LinearInterpolate(A=@n3, B=@n4, Alpha=@n7), B=Constant() {r=1}))
  Add(A=@n2, B=@n8)
  Constant() {r=1}
  Constant() {r=1}
  n9 = ScalarParameter() {defaultValue=1000, parameterName=ScaleTextures}
  LinearInterpolate(A=@n9, B=ScalarParameter() {defaultValue=1, parameterName=test}, Alpha=NamedRerouteUsage())
  n10 = Saturate(None=PerInstanceRandom())
  n11 = MaterialFunctionCall(
   UVs=MaterialFunctionCall(Texture Scale=@n9) {materialFunction=&ScaleUVsByCenter}.UVs
   "Rotation Angle (0-1)"=Subtract(A=Constant() {r=0.16}, B=LinearInterpolate(Alpha=@n10) {constA=-0.02, constB=0.02})
  ) {materialFunction=&CustomRotator}
  n12 = TextureObjectParameter() {parameterName=Specular, samplerType=SAMPLERTYPE_Masks, texture=/Game/Environments/Apollo/Sets/Terrain/Textures/T_Apollo_Cliff_S_02}
  n13 = ScalarParameter() {defaultValue=1000, parameterName=ScaleTextures}
  n14 = MaterialFunctionCall(Float3=StaticSwitchParameter(
    True=TextureSample(UVs=@n11.Rotated Values, Tex=@n12) {samplerType=SAMPLERTYPE_Masks}.RGB
    False=MaterialFunctionCall(TextureObject=@n12, TextureSize=@n13) {materialFunction=&WorldAlignedTexture}.Z Texture
   ) {parameterName=UseUVs}) {materialFunction=&BreakOutFloat3Components}
  NamedRerouteDeclaration(None=@n14.R) {name=Spec, nodeColor=<rgba 0 0.6953 1 1>}
  NamedRerouteDeclaration(None=@n14.B) {name=Rough, nodeColor=<rgba 0 0.6953 1 1>}
  n15 = TextureObjectParameter() {
   group=Textures
   parameterName=Normal
   samplerType=SAMPLERTYPE_Normal
   texture=/Game/Environments/Apollo/Sets/Terrain/Textures/T_Apollo_Cliff_N_02
  }
  NamedRerouteDeclaration(None=StaticSwitchParameter(
    True=TextureSample(UVs=@n11.Rotated Values, Tex=@n15) {samplerType=SAMPLERTYPE_Normal}.RGB
    False=MaterialFunctionCall(TextureObject=@n15, TextureSize=@n13) {materialFunction=&WorldAlignedNormal}.Z Texture
   ) {parameterName=UseUVs}) {name=Normals, nodeColor=<rgba 0 0.6953 1 1>}
  n16 = TextureObjectParameter() {parameterName=Diffuse, texture=/Game/Environments/Apollo/Sets/Terrain/Textures/T_Apollo_Cliff_D_02}
  n17 = StaticSwitchParameter(
   True=TextureSample(UVs=@n11.Rotated Values, Tex=@n16).RGBA
   False=MaterialFunctionCall(TextureObject=@n16, TextureSize=@n13, Export Float 4=StaticBool() {value=true}) {materialFunction=&WorldAlignedTexture}.Z Texture
  ) {parameterName=UseUVs}
  NamedRerouteDeclaration(None=ComponentMask(None=@n17) {a=true}) {name=DiffuseHeight, nodeColor=<rgba 0 0.6953 1 1>}
  NamedRerouteDeclaration(None=ComponentMask(None=@n17) {b=true, g=true, r=true}) {name=Diffuse, nodeColor=<rgba 0 0.6953 1 1>}
  n18 = TextureObjectParameter(UVs=MaterialFunctionCall("Rotation Angle (0-1)"=PerInstanceRandom()) {materialFunction=&CustomRotator}.Rotated Values) {
   group=Textures
   parameterName=DetailsNormal
   samplerType=SAMPLERTYPE_Normal
   texture=&T_Apollo_Cliff_N_02
  }
  NamedRerouteDeclaration(None=MaterialFunctionCall(TextureObject=@n18, TextureSize=ScalarParameter() {defaultValue=255, parameterName=DetailsScale}) {materialFunction=&WorldAlignedNormal}.Z Texture) {name=DetailNormals, nodeColor=<rgba 0 0.6953 1 1>}
  NamedRerouteDeclaration(None=Multiply(A=@n5.Results) {constB=1.5}) {name=BaseColor, nodeColor=<rgba 0 0.6953 1 1>}
  n19 = MaterialFunctionCall(BaseNormal=@n8, AdditionalNormal=@n2) {materialFunction=&BlendAngleCorrectedNormals}
  NamedRerouteDeclaration(None=@n19.Result) {name=FinalNormals, nodeColor=<rgba 0 0.6953 1 1>}
  NamedRerouteDeclaration(None=MaterialFunctionCall!(
    Coordinates
    R_01_Scale
    R_02_Scale
    R_03_Scale
    R_04_Scale
    G_01_Scale=ScalarParameter() {defaultValue=40558.9883, parameterName=Varience Scale}
    G_02_Scale
    G_03_Scale
    G_04_Scale
   ) {materialFunction=/Game/Athena/Helios/Environments/Landscape/Materials/MaterialFunctions_Utilities/MF_VariationMasks2}.G_01) {name=SandVariance, nodeColor=<rgba 1 0 1 1>}
  n20 = NamedRerouteUsage()
  n21 = ScalarParameter() {defaultValue=1, parameterName=RoughnessTweak2}
  n22 = ScalarParameter() {defaultValue=1, parameterName=RoughnessTweak}
  n23 = NamedRerouteUsage()
  n24 = ScalarParameter() {defaultValue=1, parameterName=SpecularTweak2}
  n25 = ScalarParameter() {defaultValue=1, parameterName=SpecularTweak}
  n26 = ScalarParameter() {defaultValue=0.5, parameterName=BlendIn}
  n27 = LinearInterpolate(
   A=ScalarParameter() {defaultValue=4, parameterName=MaskDensity}
   B=ScalarParameter() {defaultValue=2, parameterName=SandVarienceDensity}
   Alpha=NamedRerouteUsage()
  )
  n28 = LinearInterpolate(B=ScalarParameter() {defaultValue=0.456, parameterName=Mask_Density_Ran_Max}, Alpha=Saturate(None=PerInstanceRandom()))
  n29 = Multiply(
   A=Max(A=@n7, B=OneMinus(None=@n26))
   B=MaterialFunctionCall(Radius=Constant() {r=0.5}, Density=Subtract(A=@n27, B=@n28)) {materialFunction=&RadialGradientExponential}.RadialGradientExponential
  )
  n30 = LinearInterpolate(A=Constant() {r=-1}, B=Constant() {r=1}, Alpha=NamedRerouteUsage())
  n31 = ScalarParameter() {parameterName=Blend In Contrast}
  n32 = Subtract(
   A=DotProduct(A=TextureCoordinate(), B=Constant2Vector() {g=-1})
   B=ScalarParameter() {defaultValue=-0.9746, group=Beach, parameterName=Percentage_BeachFront}
  )
  n33 = DotProduct(A=TextureCoordinate(), B=Constant2Vector() {g=2})
  n34 = ScalarParameter() {defaultValue=-1, group=Beach, parameterName=Percentage_BeachHead}
  n35 = Multiply(A=@n32, B=Multiply(A=Subtract(A=@n33, B=@n34), B=ScalarParameter() {defaultValue=1, group=Beach, parameterName=week3_Multi}))
  n36 = StaticSwitchParameter(
   True=Multiply(A=ScalarParameter() {defaultValue=1, parameterName=Opacity, sliderMax=1}, B=NamedRerouteUsage())
   False=Multiply(
    A=MaterialFunctionCall(Transition Phase=@n29, Height Texture=@n30, Contrast=@n31) {materialFunction=&HeightLerp}.Alpha
    B=StaticSwitchParameter(True=@n35, False=@n32) {group=Beach, parameterName=Week3, sortPriority=31}
   )
  ) {parameterName="Use Opacity "}
  n37 = ScalarParameter() {defaultValue=2, parameterName=Cheat_Opacity}
  n38 = SetMaterialAttributes!(
   MaterialAttributes
   Base Color=NamedRerouteUsage()
   Roughness=LinearInterpolate(A=Multiply(A=@n20, B=@n21), B=Multiply(A=@n20, B=@n22), Alpha=@n7)
   Specular=LinearInterpolate(A=Multiply(A=@n23, B=@n24), B=Multiply(A=@n23, B=@n25), Alpha=@n7)
   Normal=NamedRerouteUsage()
   Opacity=Saturate(None=Multiply(A=@n36, B=@n37))
  ) {attributeSetTypes [
    #1
    #2
    #3
    #4
    #5
   ]}
  n39 = GetMaterialAttributes(None=@n38) {attributeGetTypes [#4]}
  n40 = SetMaterialAttributes!(MaterialAttributes=@n39.MaterialAttributes, Normal=Transform(None=@n39.Normal)) {attributeSetTypes [#4]}
  n41 = GetMaterialAttributes(None=@n40) {attributeGetTypes [
    #1
    #3
    #2
    #4
    #5
   ]}
  n42 = ScalarParameter() {defaultValue=0.15, parameterName=Mobile Desaturation}
  n43 = Multiply(
   A=Power(Base=NamedRerouteUsage(), Exp=ScalarParameter() {defaultValue=1, group=Displacement, parameterName=Displacement_Heightmap_Power})
   B=ScalarParameter() {defaultValue=1, group=Displacement, parameterName=Displacement_Heightmap_Mod}
  )
  RuntimeVirtualTextureOutput!(
   BaseColor=ShadingPathSwitch(Default=@n41.Base Color, Mobile=Desaturation(None=@n41.Base Color, Fraction=@n42))
   Specular=@n41.Specular
   Roughness=@n41.Specular
   Normal=@n41.Specular
   WorldHeight
   Opacity=@n41.Specular
   Mask
   Displacement=LinearInterpolate(
    B=Saturate(None=@n43)
    Alpha=ScalarParameter() {defaultValue=1, group=Displacement, parameterName=Displacement_Heightmap_Amount, sortPriority=-500}
   ) {constA=0.5}
   Mask4
  )
  n44 = VirtualTextureFeatureSwitch!(
   No=@n40
   Yes=MaterialFunctionCall!(RVT=@n41.MaterialAttributes, ByPass RVT=@n40) {materialFunction=/Game/Athena/Apollo/Maps/Landscape/VT/MaterialFunctions/EnableRVT}.Result
  )
  ShadingPathSwitch(Default=@n44, Deferred=@n44, Forward=@n44, Mobile=@n44)
  StaticSwitch(
   True=LinearInterpolate(A=@n19.Result, B=@n8, Alpha=@n7)
   False=@n8
   Value=StaticBoolParameter() {defaultValue=true, parameterName=UseDetailNormals}
  )
  n45 = NamedRerouteUsage()
  StaticSwitch(
   True=MaterialFunctionCall(Base=@n45, Blend=@n5.Results) {materialFunction=&Blend_Overlay}.Result
   False=@n45
   Value=@n6
  )
 ]
}
