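# Test material for visualizing event audio, reading the EventAudio_Spectrum collection parameter and building a debug digit readout of its values.
# Use_FQ_1 to 4 blend per-band range remaps, with band 4 reusing the blue channel, and tint the readout red, green, blue or white.
# The clamped result scales the Z of a normal-based offset vector masked by inverted V, and feeds a threshold test for a yellow highlight.
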Material /Clyde/Environments/KelpLinen/BP/M_TestAudio
graph {
 properties {bAutomaticallySetUsageInEditor=false, usedShadingModels=Default Lit}
 nodes [
  CollectionParameter() {collection=/Game/Athena/Events/AudioAnalysis/MPC_Audio_Reactive_Parameters, parameterName=Amplitude_Music}
  n1 = CollectionParameter() {collection=&FortniteMaterialParameters, parameterName=EventAudio_Spectrum}
  n2 = MaterialFunctionCall!(Vector4=Multiply(A=@n1), MaximumNumberOfDigits, UVs, DebugTextLocation RG_UpperRight BA_LowerLeft, Component Spacing) {materialFunction=/Engine/Functions/Engine_MaterialFunctions02/Utility/DebugFloat4Values}
  n3 = LinearInterpolate(
   A=LinearInterpolate(
    A=ComponentMask(None=@n2.ColorCodedOutput) {b=true, g=true, r=true}
    B=Constant3Vector() {constant=<rgba 1 0 0 0>}
    Alpha=ScalarParameter() {parameterName=Use_FQ_1}
   )
   B=Constant3Vector() {constant=<rgba 0 1 0 0>}
   Alpha=ScalarParameter() {parameterName=Use_FQ_2}
  )
  n4 = Constant3Vector() {constant=<rgba 0 0 1 0>}
  n5 = ScalarParameter() {parameterName=Use_FQ_3}
  n6 = CollectionParameter() {collection=&FortniteMaterialParameters, parameterName=EventAudio_Spectrum}
  n7 = Multiply(A=@n6)
  n8 = ComponentMask(None=@n7) {r=true}
  n9 = ScalarParameter() {defaultValue=0.65, parameterName=1FQ_Input_Low}
  n10 = ScalarParameter() {defaultValue=0.82, parameterName=1FQ_Input_High}
  n11 = ScalarParameter() {defaultValue=1, parameterName=1FQ_Output_Low}
  n12 = ScalarParameter() {defaultValue=200, parameterName=1FQ_Output_High}
  n13 = LinearInterpolate(
   A=LinearInterpolate(
    A=Constant() {r=1}
    B=MaterialFunctionCall(Input=@n8, Input Low=@n9, Input High=@n10, Target Low=@n11, Target High=@n12) {materialFunction=&RemapValueRange}.Result
    Alpha=ScalarParameter() {parameterName=Use_FQ_1}
   )
   B=MaterialFunctionCall(
    Input=ComponentMask(None=@n7) {g=true}
    Input Low=ScalarParameter() {defaultValue=0.45, parameterName=2FQ_Input_Low}
    Input High=ScalarParameter() {defaultValue=0.65, parameterName=2FQ_Input_High}
    Target Low=ScalarParameter() {defaultValue=1, parameterName=2FQ_Output_Low}
    Target High=ScalarParameter() {defaultValue=200, parameterName=2FQ_Output_High}
   ) {materialFunction=&RemapValueRange}.Result
   Alpha=ScalarParameter() {parameterName=Use_FQ_2}
  )
  n14 = MaterialFunctionCall(
   Input=ComponentMask(None=@n7) {b=true}
   Input Low=ScalarParameter() {defaultValue=0.23, parameterName=3FQ_Input_Low}
   Input High=ScalarParameter() {defaultValue=0.45, parameterName=3FQ_Input_High}
   Target Low=ScalarParameter() {defaultValue=1, parameterName=3FQ_Output_Low}
   Target High=ScalarParameter() {defaultValue=200, parameterName=3FQ_Output_High}
  ) {materialFunction=&RemapValueRange}
  n15 = ScalarParameter() {parameterName=Use_FQ_3}
  n16 = ComponentMask(None=@n7) {b=true}
  n17 = ScalarParameter() {defaultValue=0.2, parameterName=4FQ_Input_Low}
  n18 = ScalarParameter() {defaultValue=0.5, parameterName=4FQ_Input_High}
  n19 = ScalarParameter() {defaultValue=1, parameterName=4FQ_Output_Low}
  n20 = ScalarParameter() {defaultValue=200, parameterName=4FQ_Output_High}
  n21 = Clamp(None=LinearInterpolate(
    A=LinearInterpolate(A=@n13, B=@n14.Result, Alpha=@n15)
    B=MaterialFunctionCall(Input=@n16, Input Low=@n17, Input High=@n18, Target Low=@n19, Target High=@n20) {materialFunction=&RemapValueRange}.Result
    Alpha=ScalarParameter() {parameterName=Use_FQ_4}
   )) {maxDefault=200, minDefault=-200}
  LinearInterpolate(
   A=LinearInterpolate(
    A=LinearInterpolate(A=@n3, B=@n4, Alpha=@n5)
    B=Constant3Vector() {constant=<rgba 1 1 1 0>}
    Alpha=ScalarParameter() {parameterName=Use_FQ_4}
   )
   B=Constant3Vector() {constant=<rgba 1 0.8995 0 1>}
   Alpha=MaterialFunctionCall!(A=@n21, B=Constant() {r=200}) {materialFunction=/FortUILibrary/MaterialFunctions/Basics/MF_UI_BooleanOperations}.IsGreater
  )
  n22 = AppendVector(A=Constant() {r=1}, B=Constant() {r=1})
  n23 = TextureCoordinate()
  Multiply(A=Multiply(A=AppendVector(A=@n22, B=@n21), B=VertexNormalWS()), B=ComponentMask(None=OneMinus(None=@n23)) {g=true})
  VertexColor()
 ]
}
