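# Backdrop terrain material for KelpLinen that blends two Apollo cliff layers with water, beach and grass bands set by world height.
# The transformed backdrop normal's Z drives the HeightLerp slope masks, vertex red scales the ground transition, and forward/mobile paths swap cliffs for flat grey.
# Backdrop normals are added to layer detail, and a pivot-relative sphere-masked fake light feeds emissive.
# Vertex-alpha snow and a Winter switch sit on branches outside that normal/emissive chain.
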Material /Clyde/Environments/KelpLinen/M_KelpLinen_BackDrop_P
graph {
 properties {
  bUseMaterialAttributes=true
  bUsedWithMeshDeformer=true
  bUsedWithNanite=true
  bUsedWithSkeletalMesh=true
  pixelDepthOffsetMode=PDOM_Legacy
  usedShadingModels=Default Lit
 }
 nodes [
  n1 = TextureObjectParameter() {
   group=Dirt2
   parameterName=Specular
   samplerType=SAMPLERTYPE_Masks
   texture=/Game/Environments/Apollo/Sets/Terrain/Textures/T_Apollo_Cliff_S_01
  }
  n2 = TextureObjectParameter() {
   group=Dirt2
   parameterName=Normals
   samplerType=SAMPLERTYPE_Normal
   texture=/Game/Environments/Apollo/Sets/Terrain/Textures/T_Apollo_Cliff_N_01
  }
  n3 = MaterialFunctionCall!(
   Diffuse=TextureObjectParameter() {group=Dirt2, parameterName=Diffuse, texture=/Game/Environments/Apollo/Sets/Terrain/Textures/T_Apollo_Cliff_D_01}
   Specular=@n1
   Normal=@n2
   Nomal Value=VectorParameter() {group=LayerCliff, parameterName=Cliffs N value, sortPriority=4, defaultValue=<rgba 0.5 0.5 1 0>}.RGB
   BaseColorTone Walls
   Tile=ScalarParameter() {defaultValue=195555, parameterName=Rocks 1 Scale}
  ) {materialFunction=/Game/Environments/Apollo/VS/Materials/MF_Apollo_Cliff}
  n4 = Multiply(A=TextureCoordinate(), B=ScalarParameter() {defaultValue=10.7482, parameterName=Rocks2 Scale})
  n5 = TextureObjectParameter() {group=Dirt2, parameterName=Diffuse2, texture=/Game/Environments/Apollo/Sets/Terrain/Textures/T_Apollo_Cliff_D_02}
  n6 = TextureObjectParameter() {
   group=Dirt2
   parameterName=Specular2
   samplerType=SAMPLERTYPE_Masks
   texture=/Game/Environments/Apollo/Sets/Terrain/Textures/T_Apollo_Cliff_S_02
  }
  n7 = TextureObjectParameter() {
   group=Dirt2
   parameterName=Normals2
   samplerType=SAMPLERTYPE_Normal
   texture=/Game/Environments/Apollo/Sets/Terrain/Textures/T_Apollo_Cliff_N_02
  }
  n8 = TextureSampleParameter2D() {parameterName=Normal, samplerType=SAMPLERTYPE_Normal, texture=/Game/Athena/Apollo/Environments/Landscape/Background/BackDrop_1_N}
  n9 = ComponentMask(None=Transform(None=@n8.RGB)) {b=true}
  n10 = BlendMaterialAttributes(
   A=@n3.Result 3D
   B=SetMaterialAttributes!(
    MaterialAttributes
    Base Color=TextureSample(UVs=@n4, Tex=@n5) {samplerType=SAMPLERTYPE_Masks, texture=&T_Apollo_Cliff_S_02}.RGB
    Roughness=TextureSample(UVs=@n4, Tex=@n6) {samplerType=SAMPLERTYPE_Masks, texture=&T_Apollo_Cliff_S_02}.B
    Normal=TextureSample(UVs=@n4, Tex=@n7) {samplerType=SAMPLERTYPE_Masks, texture=&T_Apollo_Cliff_S_02}.RGB
   ) {attributeSetTypes [#1, #2, #3]}
   Alpha=MaterialFunctionCall(
    Transition Phase=ScalarParameter() {defaultValue=0.3802, parameterName=Transition Rocks}
    Height Texture=@n9
    Contrast=ScalarParameter() {parameterName=Contrast rocks}
   ) {materialFunction=&HeightLerp}.Alpha
  )
  n11 = SetMaterialAttributes!(MaterialAttributes, Base Color=Constant() {r=0.25}, Roughness) {attributeSetTypes [#1, #2]}
  n12 = SetMaterialAttributes!(
   MaterialAttributes
   Base Color=VectorParameter() {group=Beach, parameterName=Water, defaultValue=<rgba 0.0346 0.081 0.1719 0>}.RGB
   Specular=Constant() {r=0.2}
   Roughness=Constant()
  ) {attributeSetTypes [#1, #4, #2]}
  n13 = SetMaterialAttributes!(
   MaterialAttributes
   Base Color=VectorParameter() {group=Beach, parameterName=BeachLine, defaultValue=<rgba 0.5781 0.5188 0.3735 0>}.RGB
   Specular=Constant() {r=0.3}
  ) {attributeSetTypes [#1, #4]}
  n14 = ComponentMask(None=WorldPosition().XYZ) {b=true}
  n15 = Saturate(None=Subtract(A=@n14, B=Constant() {r=-5000}))
  n16 = WorldPosition() {worldPositionShaderOffset=WPT_ExcludeAllShaderOffsets}
  n17 = TextureSample(UVs=Divide(A=ComponentMask(None=@n16.XYZ) {g=true, r=true}, B=ScalarParameter() {defaultValue=850000, parameterName=GrassDetailTexture-Size})) {
   samplerSource=SSM_Wrap_WorldGroupSettings
   samplerType=SAMPLERTYPE_Grayscale
   texture=/Game/Athena/Apollo/Environments/Landscape/Materials/Biome/CommonMaterials/T_ApolloGrass_ColorMask
  }
  n18 = Subtract(A=ComponentMask(None=WorldPosition().XYZ) {b=true}, B=ScalarParameter() {defaultValue=-5500, group=Grass, parameterName=GrassTone Height})
  n19 = Constant() {r=155555}
  n20 = LinearInterpolate(
   A=VectorParameter() {group=Grass, parameterName=Grass Low, defaultValue=<rgba 0.292 0.3646 0.0019 0>}.RGB
   B=MaterialFunctionCall(
    A=VectorParameter() {group=Grass, parameterName=Grass2, defaultValue=<rgba 0.3478 0.4531 0.0696 0>}.RGB
    B=VectorParameter() {group=Grass, parameterName=Grass, defaultValue=<rgba 0.1247 0.2135 0.04 0>}.RGB
    Transition Phase=ScalarParameter() {defaultValue=0.65, group=Grass, parameterName=Grass Patern Trans}
    Height Texture=@n17.B
    Contrast=ScalarParameter() {defaultValue=1, group=Grass, parameterName=Grass Pattern Contrast}
   ) {materialFunction=&HeightLerp}.Results
   Alpha=Saturate(None=Divide(A=@n18, B=@n19))
  )
  n21 = ScalarParameter() {defaultValue=0.8, group=Grass, parameterName=Grass Roughness}
  n22 = Divide(A=Subtract(A=@n14, B=Constant() {r=-3200}), B=Constant() {r=500})
  n23 = ScalarParameter() {defaultValue=0.4911, parameterName=Transition Grass}
  n24 = VertexColor()
  n25 = BlendMaterialAttributes(
   A=ShadingPathSwitch(Default=@n10, Deferred=@n10, Forward=@n11, Mobile=@n11)
   B=BlendMaterialAttributes(
    A=BlendMaterialAttributes(A=@n12, B=@n13, Alpha=@n15)
    B=SetMaterialAttributes!(MaterialAttributes, Base Color=@n20, Roughness=@n21) {attributeSetTypes [#1, #2]}
    Alpha=Saturate(None=@n22)
   )
   Alpha=MaterialFunctionCall(Transition Phase=Multiply(A=@n23, B=@n24.R), Height Texture=@n9, Contrast=ScalarParameter() {defaultValue=5, parameterName=Contrast Grass}) {materialFunction=&HeightLerp}.Alpha
  )
  n26 = GetMaterialAttributes(None=@n25) {attributeGetTypes [#1]}
  SetMaterialAttributes!(
   MaterialAttributes=@n26.MaterialAttributes
   Base Color=MaterialFunctionCall(
    A=@n26.Base Color
    B=VectorParameter() {group=Snow, parameterName="Snow ", defaultValue=<rgba 0.8 0.8 0.8 0>}.RGB
    Transition Phase=VertexColor().A
    Height Texture=@n3.Height map 3D
    Contrast=Constant() {r=-12}
   ) {materialFunction=&HeightLerp}.Results
  ) {attributeSetTypes [#1]}
  n27 = GetMaterialAttributes(None=@n25) {attributeGetTypes [#3]}
  n28 = TextureSampleParameter2D() {parameterName=Normal, samplerType=SAMPLERTYPE_Normal, texture=&BackDrop_1_N}
  n29 = Add(A=Multiply(A=@n27.Normal, B=Constant3Vector() {constant=<rgba 1 1 0 0>}), B=@n28.RGB)
  n30 = Add(
   A=Transform(None=VectorParameter() {group=Top light, parameterName=CenterPoint, sortPriority=10, defaultValue=<rgba 0 0 0 1>}.RGB) {transformSourceType=TRANSFORMSOURCE_Local}
   B=MaterialFunctionCall() {materialFunction=&ComponentPivotLocation}.Result
  )
  n31 = SphereMask(
   A=@n30
   B=WorldPosition() {worldPositionShaderOffset=WPT_ExcludeAllShaderOffsets}.XYZ
   Radius=Multiply(
    A=MaterialFunctionCall() {materialFunction=&ObjectScale}.Scale X
    B=ScalarParameter() {defaultValue=448, group=Top light, parameterName=Mask Radius, sortPriority=25}
   )
   Hardness=ScalarParameter() {defaultValue=50, group=Top light, parameterName=Mask_Contrast}
  )
  n32 = Normalize(VectorInput=@n29)
  n33 = DotProduct(A=Normalize(VectorInput=CameraVectorWS()), B=@n32)
  n34 = Multiply(
   A=Power(Base=Saturate(None=@n33), Exp=ScalarParameter() {defaultValue=1, group=Top light, parameterName=Fresnel Contrast, sortPriority=36})
   B=ScalarParameter() {group=Top light, parameterName=Fresnel Brightness, sortPriority=37}
  )
  n35 = Subtract(A=@n30, B=WorldPosition().XYZ)
  n36 = Saturate(None=DotProduct(A=Normalize(VectorInput=@n35), B=@n32))
  n37 = Normalize(VectorInput=Subtract(A=@n30, B=WorldPosition().XYZ))
  n38 = Normalize(VectorInput=VertexNormalWS())
  n39 = LinearInterpolate(B=Saturate(None=DotProduct(A=@n37, B=@n38)), Alpha=StaticSwitch(True=Constant() {r=1}, False=Constant(), Value=StaticBool() {value=true})) {constA=1}
  n40 = Multiply(
   A=Add(A=Saturate(None=@n34), B=LinearInterpolate(B=@n36, Alpha=@n39))
   B=VectorParameter() {group=Top light, parameterName=Color, sortPriority=20, defaultValue=<rgba 1 0.263 0.032 1>}.RGB
  ) {desc=Color}
  n41 = ScalarParameter() {group=Top light, parameterName=light intensity}
  SetMaterialAttributes!(
   MaterialAttributes=SetMaterialAttributes!(MaterialAttributes=@n27.MaterialAttributes, Normal=@n29) {attributeSetTypes [#3]}
   Emissive Color=Multiply(A=Saturate(None=@n31), B=Multiply(A=@n40, B=@n41))
  ) {attributeSetTypes [#5]}
  StaticSwitchParameter(True=VectorParameter() {group=Snow, parameterName="Snow ", defaultValue=<rgba 0.8 0.8 0.8 0>}.RGB, False=@n20) {parameterName=Winter}
 ]
}
