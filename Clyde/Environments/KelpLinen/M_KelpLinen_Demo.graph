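# Masked, lit demo material for the KelpLinen environment that dissolves by camera location and height, with an optional blue eye glow.
# Engine grid textures at several tilings build the surface shading, blended by camera depth fade.
# A radial gradient of camera XY around a hardcoded world point multiplies a scale-aware bottom fade from the object pivot.
# The product drives DitherTemporalAA opacity, which the Debug_Opacity switch forces to solid.
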Material /Clyde/Environments/KelpLinen/M_KelpLinen_Demo
graph {
 properties {bAutomaticallySetUsageInEditor=false, blendMode=BLEND_Masked, usedShadingModels=Default Lit}
 nodes [
  n1 = Divide(A=TextureCoordinate())
  n2 = TextureSample(UVs=@n1) {samplerType=SAMPLERTYPE_LinearColor, texture=/Engine/EngineMaterials/T_Default_Material_Grid_M}
  n3 = TextureSample(UVs=Divide(A=@n1) {constB=0.1}) {samplerType=SAMPLERTYPE_LinearColor, texture=&T_Default_Material_Grid_M}
  n4 = MaterialFunctionCall(Fade Length=Constant() {r=1500}) {materialFunction=&CameraDepthFade}
  n5 = LinearInterpolate(A=LinearInterpolate(Alpha=@n3.G) {constA=0.7}, B=Constant() {r=1}, Alpha=@n4.Result)
  Clamp(None=Add(A=LinearInterpolate(Alpha=@n2.G) {constB=0.5}, B=Multiply(A=@n5) {constA=1}))
  LinearInterpolate(A=@n3.A, B=@n2.A, Alpha=@n4.Result) {desc=Uses Cavity mask near camera}
  n6 = LinearInterpolate(Alpha=TextureSample(UVs=TextureCoordinate() {uTiling=20, vTiling=20}) {samplerType=SAMPLERTYPE_LinearColor, texture=&T_Default_Material_Grid_M}.R) {constA=0.4}
  n7 = Constant() {r=0.295}
  n8 = Constant() {r=0.66}
  n9 = LinearInterpolate(A=@n3.G, Alpha=@n4.Result) {constB=0, desc=Add small tiles to basecolor near camera}
  n10 = LinearInterpolate(A=ComponentMask(None=Add(A=@n2.G, B=@n9) {constB=0}) {r=true}, Alpha=Constant() {r=0.5}) {constAlpha=0.875, constB=0.5}
  Multiply(
   A=LinearInterpolate(A=@n6, B=OneMinus(None=@n6), Alpha=LinearInterpolate(A=@n3.G, Alpha=@n4.Result))
   B=Multiply(A=LinearInterpolate(A=@n7, B=@n8, Alpha=@n10), B=Constant() {r=0.5}) {constB=0.5}
  )
  Multiply(
   A=TextureSample(UVs=Divide(A=@n1) {constB=0.05}) {samplerType=SAMPLERTYPE_Normal, texture=/Engine/EngineMaterials/T_Default_Material_Grid_N}.RGB
   B=Constant3Vector() {constant=<rgba 0.3 0.3 1 0>}
  )
  n11 = ObjectPositionWS()
  n12 = MaterialFunctionCall() {materialFunction=&ObjectScale}
  n13 = Subtract(
   A=Subtract(A=WorldPosition().Z, B=ComponentMask(None=@n11) {b=true})
   B=Multiply(A=@n12.Scale X, B=ScalarParameter() {defaultValue=-70, parameterName=BottomFade_Offset})
  )
  n14 = Multiply(A=@n12.Scale X, B=ScalarParameter() {defaultValue=1166.9116, parameterName=BottomFade_Distance})
  NamedRerouteDeclaration(None=Saturate(None=Divide(A=@n13, B=@n14))) {name=BottomFade, nodeColor=<rgba 1 0 0.8672 1>}
  n15 = Constant3Vector() {constant=<rgba 0 0 1 1>}
  n16 = ScalarParameter() {defaultValue=100, parameterName=Str}
  StaticSwitchParameter(
   True=MaterialFunctionCall(Emissive=Multiply(A=@n15, B=@n16)) {materialFunction=&MF_FX_EmissiveAdjustments}.Result
   False=Constant()
  ) {parameterName=eye_Glow}
  n17 = ScalarParameter() {defaultValue=1, parameterName=Max_Opacity}
  n18 = MaterialFunctionCall(
   UVs=ComponentMask(None=CameraPositionWS()) {g=true, r=true}
   CenterPosition=ComponentMask(None=Constant3Vector() {constant=<rgba -31136 -105728 -1920 1>}) {g=true, r=true}
   Radius=ScalarParameter() {defaultValue=23054.5273, parameterName=Radius}
   Density=ScalarParameter() {defaultValue=1.7342, parameterName=Density}
  ) {materialFunction=&RadialGradientExponential}
  n19 = ScalarParameter() {defaultValue=2, parameterName=Density_Multi}
  n20 = Multiply(A=Saturate(None=Multiply(A=@n18.RadialGradientExponential, B=@n19)), B=NamedRerouteUsage())
  StaticSwitchParameter(
   True=Constant() {r=1}
   False=MaterialFunctionCall(Alpha Threshold=LinearInterpolate(B=@n17, Alpha=@n20), Random=Constant() {r=1}) {materialFunction=&DitherTemporalAA}.Result
  ) {parameterName=Debug_Opacity}
 ]
}
