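# Subsurface material for kelp snow tops, with parameters for color, subsurface color, roughness and specular, plus a fake colored "Top light" glow.
# A local-space center point offset from the component pivot drives a sphere mask whose radius follows object scale.
# Inside it, light direction against world up and vertex normal, plus a view-versus-up "Fresnel" term, shade the emissive color.
# Light intensity and Fresnel brightness default to zero.
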Material /Clyde/Environments/KelpLinen/M_Kelp_SnowTops
graph {
 properties {pixelDepthOffsetMode=PDOM_Legacy, shadingModel=MSM_Subsurface, usedShadingModels=Subsurface}
 nodes [
  n1 = VectorParameter() {group=Top light, parameterName=CenterPoint, sortPriority=10, defaultValue=<rgba 367.2455 -0.7016 903.2573 1>}
  n2 = Add(
   A=Transform(None=@n1.RGB) {transformSourceType=TRANSFORMSOURCE_Local}
   B=MaterialFunctionCall() {materialFunction=&ComponentPivotLocation}.Result
  )
  n3 = WorldPosition() {worldPositionShaderOffset=WPT_ExcludeAllShaderOffsets}
  n4 = Multiply(
   A=MaterialFunctionCall() {materialFunction=&ObjectScale}.Scale X
   B=ScalarParameter() {defaultValue=402.7367, group=Top light, parameterName=Mask Radius, sortPriority=25}
  )
  n5 = ScalarParameter() {defaultValue=2, group=Top light, parameterName=Mask_Contrast}
  n6 = Normalize(VectorInput=CameraVectorWS())
  n7 = Normalize(VectorInput=Constant3Vector() {constant=<rgba 0 0 1 0>})
  n8 = Power(
   Base=Saturate(None=DotProduct(A=@n6, B=@n7))
   Exp=ScalarParameter() {defaultValue=1, group=Top light, parameterName=Fresnel Contrast, sortPriority=36}
  )
  n9 = ScalarParameter() {group=Top light, parameterName=Fresnel Brightness, sortPriority=37}
  n10 = WorldPosition()
  n11 = DotProduct(A=Normalize(VectorInput=Subtract(A=@n2, B=@n10.XYZ)), B=@n7)
  n12 = Subtract(A=@n2, B=WorldPosition().XYZ)
  n13 = VertexNormalWS()
  n14 = Saturate(None=DotProduct(A=Normalize(VectorInput=@n12), B=Normalize(VectorInput=@n13)))
  n15 = StaticSwitch(True=Constant() {r=1}, False=Constant(), Value=StaticBool() {value=true})
  n16 = Add(A=Saturate(None=Multiply(A=@n8, B=@n9)), B=LinearInterpolate(B=Saturate(None=@n11), Alpha=LinearInterpolate(B=@n14, Alpha=@n15) {constA=1}))
  n17 = VectorParameter() {group=Top light, parameterName=Color_Emissive, sortPriority=20, defaultValue=<rgba 1 0 0.2283 1>}
  Multiply(
   A=Saturate(None=SphereMask(A=@n2, B=@n3.XYZ, Radius=@n4, Hardness=@n5))
   B=Multiply(A=Multiply(A=@n16, B=@n17.RGB) {desc=Color}, B=ScalarParameter() {group=Top light, parameterName=light intensity})
  )
  ScalarParameter() {defaultValue=0.7, parameterName=Roughness}
  ScalarParameter() {defaultValue=0.75, parameterName=Specular}
  VectorParameter() {parameterName=Color, sortPriority=20, defaultValue=<rgba 0.7917 0.7917 0.7917 0>}
  VectorParameter() {parameterName=SSSColor, defaultValue=<rgba 0.5364 0.8367 0.8802 0>}
 ]
}
