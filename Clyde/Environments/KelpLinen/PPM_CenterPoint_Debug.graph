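# PPM_CenterPoint_Debug is an unlit post-process debug material that visualizes a center point and radius in the world by tinting the scene.
# World-position XY feeds RadialGradientExponential around the Location parameter, with Radius and Density shaping the falloff, so height is ignored.
# The gradient is scaled by Density_Multi, saturated and used to blend scene color toward the lerp's default white.
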Material /Clyde/Environments/KelpLinen/PPM_CenterPoint_Debug
graph {
 properties {bAutomaticallySetUsageInEditor=false, materialDomain=MD_PostProcess, shadingModel=MSM_Unlit, usedShadingModels=Unlit}
 nodes [
  n1 = MaterialFunctionCall(
   UVs=WorldPosition().XY
   CenterPosition=ComponentMask(None=VectorParameter() {parameterName=Location, defaultValue=<rgba -31136 -105728 -1920 1>}.RGB) {g=true, r=true}
   Radius=ScalarParameter() {defaultValue=23054.5273, parameterName=Radius}
   Density=ScalarParameter() {defaultValue=1.7342, parameterName=Density}
  ) {materialFunction=&RadialGradientExponential}
  n2 = ScalarParameter() {defaultValue=2, parameterName=Density_Multi}
  LinearInterpolate(A=SceneTexture() {sceneTextureId=PPI_PostProcessInput0}.Color, Alpha=Saturate(None=Multiply(A=@n1.RadialGradientExponential, B=@n2)))
 ]
}
